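# Physics asset for the Pumpkin Slice Commando hat accessory, defining collision and jointing for its dangling stem.
# It has two sphere bodies: a large one on the head bone and a small one on the dyn_stem1 bone.
# A single constraint links the stem to the head with very tight swing and twist limits, allowing only slight wobble.
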PhysicsAsset /Game/Accessories/Hats/M_MED_Commando_PumpkinSlice_Hats/Meshes/M_MED_Commando_PumpkinSlice_Hats_Physics
bodies {
 dyn_stem1 [{shapeType=Sphere, radius=1.9715, center=<-0.1364 0.1578 0>}]
 head [{shapeType=Sphere, radius=14.2115, center=<2.9831 3.1364 0.3157>}]
}
constraints [{bone1Name=dyn_stem1, bone2Name=head, swing1LimitDegrees=2, swing2LimitDegrees=2, twistLimitDegrees=2}]
